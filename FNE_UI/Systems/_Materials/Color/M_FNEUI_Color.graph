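# Translucent UI material that provides per-state color and opacity for widget states: idle, hovered, focused, pressed and disabled.
# Each state's color parameter feeds MF_FNEUI_ButtonStatesBase_Lerp, with an activation switch and transition scalar for every non-idle state.
# Activations are scalar parameters, so unused states can stay off.
# Transition parameter names are fixed because parent widgets rely on them; the disabled one is TransDisabled, without an underscore.
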Material /FNE_UI/Systems/_Materials/Color/M_FNEUI_Color
graph {
 properties {blendMode=BLEND_Translucent, materialDomain=MD_UI, pixelDepthOffsetMode=PDOM_Legacy, usedShadingModels=Default Lit}
 nodes [
  n1 = VectorParameter() {desc="Base color in idle state\r\n", group=State Colors, parameterName=Color and Opacity, sortPriority=0, defaultValue=<rgba 1 1 1 1>}
  n2 = ScalarParameter() {
   desc="0 (<1) = Not active\r\n1 = Active\r\n\r\nWhen Active, The following parameters will be applied for the active state.\r\n\r\nWhen Not active, whatever the state parameter value, it will not be applied.\r\n\r\nUseful to set the states you are using only.\r\n"
   group=State Activations
   parameterName=Activate Focused State
   sliderMax=1
   sortPriority=2
  }
  n3 = VectorParameter() {group=State Colors, parameterName=Focused Color and Opacity, sortPriority=2, defaultValue=<rgba 1 1 1 1>}
  n4 = ScalarParameter() {
   desc="WARNING : \r\nThe name of this parameter must not be edited. It is used in many widget parenting. Changing its name will brake the transition behavior.\r\n"
   group=State Transitions
   parameterName=Trans_Focused
   sliderMax=1
   sortPriority=2
  }
  n5 = ScalarParameter() {
   desc="0 (<1) = Not active\r\n1 = Active\r\n\r\nWhen Active, The following parameters will be applied for the active state.\r\n\r\nWhen Not active, whatever the state parameter value, it will not be applied.\r\n\r\nUseful to set the states you are using only.\r\n"
   group=State Activations
   parameterName=Activate Hovered State
   sliderMax=1
   sortPriority=1
  }
  n6 = VectorParameter() {group=State Colors, parameterName=Hovered Color and Opacity, sortPriority=1, defaultValue=<rgba 1 1 1 1>}
  n7 = ScalarParameter() {
   desc="WARNING : \r\nThe name of this parameter must not be edited. It is used in many widget parenting. Changing its name will brake the transition behavior.\r\n"
   group=State Transitions
   parameterName=Trans_Hovered
   sliderMax=1
   sortPriority=1
  }
  n8 = ScalarParameter() {
   desc="0 (<1) = Not active\r\n1 = Active\r\n\r\nWhen Active, The following parameters will be applied for the active state.\r\n\r\nWhen Not active, whatever the state parameter value, it will not be applied.\r\n\r\nUseful to set the states you are using only.\r\n"
   group=State Activations
   parameterName=Activate Pressed State
   sliderMax=1
   sortPriority=3
  }
  n9 = VectorParameter() {group=State Colors, parameterName=Pressed Color and Opacity, sortPriority=3, defaultValue=<rgba 1 1 1 1>}
  n10 = ScalarParameter() {
   desc="WARNING : \r\nThe name of this parameter must not be edited. It is used in many widget parenting. Changing its name will brake the transition behavior.\r\n"
   group=State Transitions
   parameterName=Trans_Pressed
   sliderMax=1
   sortPriority=3
  }
  n11 = ScalarParameter() {
   desc="0 (<1) = Not active\r\n1 = Active\r\n\r\nWhen Active, The following parameters will be applied for the active state.\r\n\r\nWhen Not active, whatever the state parameter value, it will not be applied.\r\n\r\nUseful to set the states you are using only.\r\n"
   group=State Activations
   parameterName=Activate Disabled State
   sliderMax=1
   sortPriority=4
  }
  n12 = VectorParameter() {group=State Colors, parameterName=Disabled Color and Opacity, sortPriority=4, defaultValue=<rgba 1 1 1 1>}
  n13 = ScalarParameter() {
   desc="WARNING : \r\nThe name of this parameter must not be edited. It is used in many widget parenting. Changing its name will brake the transition behavior.\r\n"
   group=State Transitions
   parameterName=TransDisabled
   sliderMax=1
   sortPriority=4
  }
  MaterialFunctionCall!(
   Idle=@n1.RGB
   Idle Opacity=@n1.A
   Is Focused Active=@n2
   Focused=@n3.RGB
   Focused Opacity=@n3.A
   Trans_Focused=@n4
   Is Hovered Active=@n5
   Hovered=@n6.RGB
   Hovered Opacity=@n6.A
   Trans_Hovered=@n7
   IsPressed Active=@n8
   Pressed=@n9.RGB
   Pressed Opacity=@n9.A
   Trans_Pressed=@n10
   Is Disabled Active=@n11
   Disabled=@n12.RGB
   Disabled Opacity=@n12.A
   Trans_Disabled=@n13
  ) {materialFunction=/FNE_UI/Systems/_MaterialFunctions/ButtonStates/MF_FNEUI_ButtonStatesBase_Lerp}
 ]
}
